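; Calendar-driven rift event handler that makes placed props appear or disappear in Athena.
; After match start, the server caches prop classes and transforms, pre-removing those due to appear.
; A precursor flag drives shimmer, telegraph and feedback effects; the execution flag or ActivateEarly triggers spawning or destruction, with a timer clearing overlapping building actors.
; RepNotifies replicate effects and destruction; clients also watch the calendar to hide disappearing props.
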(event Custom|ActivateEarly
  (bind _destroyoverlappingprops (Variables|CalendarPropHandler|GetDestroyOverlappingProps))
  (Networking|SwitchHasAuthority
    (:Authority
      (if (== (Utilities|Array|Length (Variables|Default|GetPropInfo)) 0)
        ;[1 1 1] store prop info
        (for _array_element (Variables|CalendarPropHandler|GetPropReferences)
          (if (Utilities|IsValid _array_element :!class /Script/Engine.KismetSystemLibrary)
            (Utilities|Array|AddUnique (Variables|Default|GetPropInfo) (Utilities|Struct|MakeCalendarEventPropInfo (Utilities|GetClass _array_element) (Transformation|GetActorTransform _array_element)))
            (else
              (Development|PrintString "Prop references not valid ")))))
      ;[1 1 1] rift event - handle spawning/destruction of actor
      (Utilities|FlowControl|DoOnce false
        (:Completed
          (switch Utilities|FlowControl|Switch|SwitchonCalendarPropHandler_EventTypes (Variables|CalendarPropHandler|GetEventType)
            (:NewEnumerator0
              (if (not _destroyoverlappingprops)
                (CallFunction|SpawnProp(Server))))
            (:NewEnumerator1
              (if (not _destroyoverlappingprops)
                (Variables|Default|SetDestroyProp true)))
            (:NewEnumerator2
              (Development|PrintString "random event logic goes here if we get to it")))
          (Utilities|Time|SetTimerbyFunctionName 0 "DestroyOverlappingBuildingProps" 0.01 true))))))

(event EventBeginPlay
  (bind _executioneventflag (Variables|CalendarPropHandler|GetExecutionEventFlag))
  ;[1 0 0] beginplay
  (Networking|SwitchHasAuthority
    (:Authority
      ;[1 1 1] Wait for match started.
      (bind _asfort_game_mode_zone (Utilities|Casting|CastToFortGameModeZone (Game|GetGameMode)))
      (EventDispatchers|BindEventtoOnHandleMatchHasStarted _asfort_game_mode_zone (EventDispatchers|CreateEvent :!event Custom|BeginScript)))
    (:Remote
      ;[1 1 1] On client, hide disappearing items when the calendar event is active.
      (switch Utilities|FlowControl|Switch|SwitchonCalendarPropHandler_EventTypes (Variables|CalendarPropHandler|GetEventType)
        (:NewEnumerator1
          (if (not (Utilities|String|EqualExactly(String) _executioneventflag))
            (MCPContext|WatchCalendarEvent _executioneventflag
              (:OnEventActive
                (for _array_element (Variables|CalendarPropHandler|GetPropReferences)
                  (Utilities|IsValid _array_element
                    (:Is%20Valid
                      (Utilities|Casting|CastToBP_SVLair_GeodeObject _array_element
                        (:CastFailed
                          (Rendering|SetActorHiddenInGame _array_element true))))))))))))))

(event Custom|TriggerManualDebug
  (bind _silentevent (Variables|CalendarPropHandler|GetSilentEvent))
  (bind _debuginterval (Variables|CalendarPropHandler|GetDebugInterval))
  (bind _telegraphduration (Variables|Default|GetTelegraphDuration))
  (bind _precursoreventflag (Variables|CalendarPropHandler|GetPrecursorEventFlag))
  (bind _destroyoverlappingprops (Variables|CalendarPropHandler|GetDestroyOverlappingProps))
  (bind _executioneventflag (Variables|CalendarPropHandler|GetExecutionEventFlag))
  ;[1 1 1] manual debug
  (if (Variables|CalendarPropHandler|GetManualDebug)
    ;[1 1 1] is this a random event and should forego prop related logic?
    (if (not (Utilities|Enum|Equal(Enum) (Variables|CalendarPropHandler|GetEventType) "NewEnumerator2"))
      ;[1 1 1] store prop info
      (for _array_element (Variables|CalendarPropHandler|GetPropReferences)
        (if (Utilities|IsValid _array_element :!class /Script/Engine.KismetSystemLibrary)
          (Utilities|Array|AddUnique (Variables|Default|GetPropInfo) (Utilities|Struct|MakeCalendarEventPropInfo (Utilities|GetClass _array_element) (Transformation|GetActorTransform _array_element)))
          (else
            (Development|PrintString (Utilities|String|Append "Prop references not valid " (Utilities|GetDisplayName _array_element))))))
      ;[1 1 1] destroy prop if it's supposed to appear
      (switch Utilities|FlowControl|Switch|SwitchonCalendarPropHandler_EventTypes (Variables|CalendarPropHandler|GetEventType)
        (:NewEnumerator0
          (if (not (Variables|Default|GetPropHasBeenSpawned))
            (for _array_element_2 (Variables|CalendarPropHandler|GetPropReferences)
              (bind _asbuilding_actor (Utilities|Casting|CastToBuildingActor _array_element_2))
              (Gameplay|SilentDie _asbuilding_actor false :!class /Script/FortniteGame.BuildingActor))))
        (:NewEnumerator1
          _))
      ;[1 1 1] shimmer event - handle tells
      (if (Variables|CalendarPropHandler|GetDebugMode)
        (if _silentevent
          (Utilities|FlowControl|Delay _debuginterval)
          (else
            (Variables|Default|SetIsShimmerEnabled 1)
            (Utilities|FlowControl|Delay _debuginterval)
            (CallFunction|RiftEventTelegraph)
            (Utilities|FlowControl|Delay (Variables|Default|GetTelegraphDuration))
            (Variables|Default|SetIsShimmerEnabled 0)
            (CallFunction|RiftEventFeedback)))
        ;[1 1 1] rift event - handle spawning/destruction of actor
        (if (Variables|CalendarPropHandler|GetDebugMode)
          _
          (elif (Utilities|String|EqualExactly(String) _executioneventflag)
            (Development|PrintString "Error: Calendar event is null")
            (else
              (MCPContext|WatchCalendarEvent _executioneventflag
                (:OnEventActive
                  (Utilities|Casting|CastToBP_SVLair_GeodeObject (Utilities|Array|Get(acopy) (Variables|CalendarPropHandler|GetPropReferences))
                    (:then
                      (Utilities|FlowControl|Delay 5.0) ; Add a delay so the Calendar doesn't destroy the rocket a little too early.
                    )
                    (:CastFailed
                      _)))))))
        (Utilities|FlowControl|DoOnce false
          (:Completed
            (switch Utilities|FlowControl|Switch|SwitchonCalendarPropHandler_EventTypes (Variables|CalendarPropHandler|GetEventType)
              (:NewEnumerator0
                (if (not _destroyoverlappingprops)
                  (CallFunction|SpawnProp(Server))))
              (:NewEnumerator1
                (if (not _destroyoverlappingprops)
                  (Variables|Default|SetDestroyProp true)))
              (:NewEnumerator2
                (Development|PrintString "random event logic goes here if we get to it")))
            (Utilities|Time|SetTimerbyFunctionName 0 "DestroyOverlappingBuildingProps" 0.01 true)))
        (elif _silentevent
          _
          (elif (Utilities|String|EqualExactly(String) _precursoreventflag)
            _
            (else
              (bind (_eventname _timeuntilend) (MCPContext|WatchCalendarEvent _precursoreventflag)
                (:then
                  _)
                (:OnEventActive
                  (Variables|Default|SetIsShimmerEnabled 1)
                  (if (and (< (- (Math|Timespan|GetTotalSeconds _timeuntilend) _telegraphduration) 3.6e+03) (>= (- (Math|Timespan|GetTotalSeconds _timeuntilend) _telegraphduration) 5.0))
                    (Utilities|FlowControl|Delay (- (Math|Timespan|GetTotalSeconds _timeuntilend) _telegraphduration))
                    (CallFunction|RiftEventTelegraph)))
                (:OnEventEnded
                  (Variables|Default|SetIsShimmerEnabled 0)
                  (CallFunction|RiftEventFeedback))))))))))

(event Custom|BeginScript
  (bind _silentevent (Variables|CalendarPropHandler|GetSilentEvent))
  (bind _debuginterval (Variables|CalendarPropHandler|GetDebugInterval))
  (bind _telegraphduration (Variables|Default|GetTelegraphDuration))
  (bind _precursoreventflag (Variables|CalendarPropHandler|GetPrecursorEventFlag))
  (bind _destroyoverlappingprops (Variables|CalendarPropHandler|GetDestroyOverlappingProps))
  (bind _executioneventflag (Variables|CalendarPropHandler|GetExecutionEventFlag))
  (if (not (Variables|CalendarPropHandler|GetSilentEvent))
    (Variables|Default|SetIsShimmerEnabled 0))
  (Utilities|FlowControl|Delay 1e+01)
  ;[1 1 1] manual debug
  (if (not (Variables|CalendarPropHandler|GetManualDebug))
    ;[1 1 1] is this a random event and should forego prop related logic?
    (if (not (Utilities|Enum|Equal(Enum) (Variables|CalendarPropHandler|GetEventType) "NewEnumerator2"))
      ;[1 1 1] store prop info
      (for _array_element (Variables|CalendarPropHandler|GetPropReferences)
        (if (Utilities|IsValid _array_element :!class /Script/Engine.KismetSystemLibrary)
          (Utilities|Array|AddUnique (Variables|Default|GetPropInfo) (Utilities|Struct|MakeCalendarEventPropInfo (Utilities|GetClass _array_element) (Transformation|GetActorTransform _array_element)))
          (else
            (Development|PrintString (Utilities|String|Append "Prop references not valid " (Utilities|GetDisplayName _array_element))))))
      ;[1 1 1] destroy prop if it's supposed to appear
      (switch Utilities|FlowControl|Switch|SwitchonCalendarPropHandler_EventTypes (Variables|CalendarPropHandler|GetEventType)
        (:NewEnumerator0
          (if (not (Variables|Default|GetPropHasBeenSpawned))
            (for _array_element_2 (Variables|CalendarPropHandler|GetPropReferences)
              (bind _asbuilding_actor (Utilities|Casting|CastToBuildingActor _array_element_2))
              (Gameplay|SilentDie _asbuilding_actor false :!class /Script/FortniteGame.BuildingActor))))
        (:NewEnumerator1
          _))
      ;[1 1 1] shimmer event - handle tells
      (if (Variables|CalendarPropHandler|GetDebugMode)
        (if _silentevent
          (Utilities|FlowControl|Delay _debuginterval)
          (else
            (Variables|Default|SetIsShimmerEnabled 1)
            (Utilities|FlowControl|Delay _debuginterval)
            (CallFunction|RiftEventTelegraph)
            (Utilities|FlowControl|Delay (Variables|Default|GetTelegraphDuration))
            (Variables|Default|SetIsShimmerEnabled 0)
            (CallFunction|RiftEventFeedback)))
        ;[1 1 1] rift event - handle spawning/destruction of actor
        (if (Variables|CalendarPropHandler|GetDebugMode)
          _
          (elif (Utilities|String|EqualExactly(String) _executioneventflag)
            (Development|PrintString "Error: Calendar event is null")
            (else
              (MCPContext|WatchCalendarEvent _executioneventflag
                (:OnEventActive
                  (Utilities|Casting|CastToBP_SVLair_GeodeObject (Utilities|Array|Get(acopy) (Variables|CalendarPropHandler|GetPropReferences))
                    (:then
                      (Utilities|FlowControl|Delay 5.0) ; Add a delay so the Calendar doesn't destroy the rocket a little too early.
                    )
                    (:CastFailed
                      _)))))))
        (Utilities|FlowControl|DoOnce false
          (:Completed
            (switch Utilities|FlowControl|Switch|SwitchonCalendarPropHandler_EventTypes (Variables|CalendarPropHandler|GetEventType)
              (:NewEnumerator0
                (if (not _destroyoverlappingprops)
                  (CallFunction|SpawnProp(Server))))
              (:NewEnumerator1
                (if (not _destroyoverlappingprops)
                  (Variables|Default|SetDestroyProp true)))
              (:NewEnumerator2
                (Development|PrintString "random event logic goes here if we get to it")))
            (Utilities|Time|SetTimerbyFunctionName 0 "DestroyOverlappingBuildingProps" 0.01 true)))
        (elif _silentevent
          _
          (elif (Utilities|String|EqualExactly(String) _precursoreventflag)
            _
            (else
              (bind (_eventname _timeuntilend) (MCPContext|WatchCalendarEvent _precursoreventflag)
                (:then
                  _)
                (:OnEventActive
                  (Variables|Default|SetIsShimmerEnabled 1)
                  (if (and (< (- (Math|Timespan|GetTotalSeconds _timeuntilend) _telegraphduration) 3.6e+03) (>= (- (Math|Timespan|GetTotalSeconds _timeuntilend) _telegraphduration) 5.0))
                    (Utilities|FlowControl|Delay (- (Math|Timespan|GetTotalSeconds _timeuntilend) _telegraphduration))
                    (CallFunction|RiftEventTelegraph)))
                (:OnEventEnded
                  (Variables|Default|SetIsShimmerEnabled 0)
                  (CallFunction|RiftEventFeedback))))))))))

(event Custom|RiftEventFeedback
  ;[1 1 1] rift event feedback RPC
  (Effects|SpawnEmitteratLocation (Variables|Default|GetFX_RiftEvent) (Transformation|GetWorldLocation (Variables|Default|GetDestructionBox)) (Transformation|GetWorldRotation (Variables|Default|GetDestructionBox)) (/ (Math|Conversions|ToVector(Float) (Math|Vector|GetMaxElement (Components|Box|GetScaledBoxExtent (Variables|Default|GetDestructionBox)))) 1.8e+02))
  (Development|PrintString "Rift Event Feedback" (Variables|CalendarPropHandler|GetDebugMode) (Variables|CalendarPropHandler|GetDebugMode)))

(event Custom|RiftEventTelegraph
  (bind _destructionbox (Variables|Default|GetDestructionBox))
  ;[1 1 1] Rift event telegraph RPC
  (Effects|SpawnEmitteratLocation (Variables|Default|GetFX_RiftTelegraphEvent) (Transformation|GetWorldLocation _destructionbox) (Transformation|GetWorldRotation _destructionbox) (/ (Math|Conversions|ToVector(Float) (Math|Vector|GetMaxElement (Components|Box|GetScaledBoxExtent _destructionbox))) 1.8e+02))
  (Audio|SpawnSoundAttached (Variables|Default|GetSound_Telegraph) _destructionbox))

(fn DestroyOverlappingBuildingProps
  (bind _outactors (Collision|ComponentOverlapActors (Variables|Default|GetDestructionBox) (Transformation|GetWorldTransform (Variables|Default|GetDestructionBox)) (Utilities|Array|MakeArray "ObjectTypeQuery1") "/Script/FortniteGame.BuildingActor"))
  (if (== (Utilities|Array|Length _outactors) 0)
    (Utilities|Time|ClearTimerbyFunctionName 0 "DestroyOverlappingBuildingProps")
    (switch Utilities|FlowControl|Switch|SwitchonCalendarPropHandler_EventTypes (Variables|CalendarPropHandler|GetEventType)
      (:NewEnumerator0
        (CallFunction|SpawnProp(Server)))
      (:NewEnumerator1
        (Variables|Default|SetDestroyProp true)))
    (else
      (bind _asbuilding_actor (Utilities|Casting|CastToBuildingActor (Utilities|Array|Get(acopy) _outactors)))
      (Gameplay|SilentDie _asbuilding_actor false :!class /Script/FortniteGame.BuildingActor))))

(fn OnRep_DestroyProp
  (for _array_element (Variables|CalendarPropHandler|GetPropReferences)
    (if (Utilities|NotEqual(Object) _array_element 0)
      (bind _asbuilding_actor (Utilities|Casting|CastToBuildingActor _array_element))
      (Gameplay|SilentDie _asbuilding_actor false :!class /Script/FortniteGame.BuildingActor)
      (Development|PrintString "Destroyed Prop!" (Variables|CalendarPropHandler|GetDebugMode) (Variables|CalendarPropHandler|GetDebugMode))
      (if (not (Variables|CalendarPropHandler|GetSilentEvent))
        (switch Utilities|FlowControl|Switch|SwitchonCalendarPropHandler_EventTypes (Variables|CalendarPropHandler|GetEventType)
          (:NewEnumerator1
            (bind _returnvalue (Effects|SpawnEmitteratLocation "/Figment_S05/Athena/Effects/Particles/Geode/P_Geode_Oddity_Leftovers.P_Geode_Oddity_Leftovers" (Transformation|GetWorldLocation (Variables|Default|GetDestructionBox)) (Transformation|GetWorldRotation (Variables|Default|GetDestructionBox)) (/ (Math|Conversions|ToVector(Float) (Math|Vector|GetMaxElement (Components|Box|GetScaledBoxExtent (Variables|Default|GetDestructionBox)))) 1.8e+02)))
            (bind _returnvalue_2 (Audio|SpawnSoundAttached (Variables|Default|GetSound_InWorld) _returnvalue))
            (Audio|Components|Audio|FadeIn _returnvalue_2 1.0 0.5 0.0 "Linear" :!class /Script/Engine.AudioComponent)))))))

(fn OnRep_IsShimmerEnabled
  (bind _shimmerfx (Variables|Default|GetShimmerFX))
  (Networking|SwitchHasAuthority
    (:Remote
      (switch int (Variables|Default|GetIsShimmerEnabled)
        (:0
          (Utilities|FlowControl|Sequence
            (:then_0
              (Utilities|IsValid _shimmerfx
                (:Is%20Valid
                  (Components|DestroyComponent _shimmerfx)
                  (Development|PrintString "Shimmer Disabled" (Variables|CalendarPropHandler|GetDebugMode) (Variables|CalendarPropHandler|GetDebugMode)))))
            (:then_1
              (bind _inworldsound (Variables|Default|GetInWorldSound)
                (:then
                  (Audio|Components|Audio|FadeOut _inworldsound 0.25 0.0 "Linear" :!class /Script/Engine.AudioComponent))
                (:else)))))
        (:1
          (bind _returnvalue (Effects|SpawnEmitteratLocation :EmitterTemplate (Variables|Default|GetFX_ShimmerEvent) :Location (Transformation|GetWorldLocation (Variables|Default|GetDestructionBox)) :Scale (/ (Math|Conversions|ToVector(Float) (* (Math|Vector|GetMaxElement (Components|Box|GetScaledBoxExtent (Variables|Default|GetDestructionBox))) (Variables|CalendarPropHandler|GetTellScale))) 1.8e+02) :bAutoDestroy false :PoolingMethod "None" :bAutoActivateSystem true))
          (bind _output_get (Variables|Default|SetShimmerFX _returnvalue))
          (Development|PrintString "Shimmer Enabled" (Variables|CalendarPropHandler|GetDebugMode) (Variables|CalendarPropHandler|GetDebugMode))
          (bind _returnvalue_2 (Audio|SpawnSoundAttached (Variables|Default|GetSound_InWorld) _output_get))
          (bind _output_get_2 (Variables|Default|SetInWorldSound _returnvalue_2))
          (Audio|Components|Audio|FadeIn _output_get_2 1.0 1.0 0.0 "Linear" :!class /Script/Engine.AudioComponent))))))

(fn OnRep_ShimmerEnabled
  (bind _shimmerfx (Variables|Default|GetShimmerFX))
  (Networking|SwitchHasAuthority
    (:Remote
      (if (Variables|Default|GetShimmerEnabled)
        (bind _returnvalue (Effects|SpawnEmitteratLocation :EmitterTemplate (Variables|Default|GetFX_ShimmerEvent) :Location (Transformation|GetWorldLocation (Variables|Default|GetDestructionBox)) :Scale (/ (Math|Conversions|ToVector(Float) (* (Math|Vector|GetMaxElement (Components|Box|GetScaledBoxExtent (Variables|Default|GetDestructionBox))) (Variables|CalendarPropHandler|GetTellScale))) 1.8e+02) :bAutoDestroy false :PoolingMethod "None" :bAutoActivateSystem true))
        (bind _output_get (Variables|Default|SetShimmerFX _returnvalue))
        (Development|PrintString "Shimmer Enabled" (Variables|CalendarPropHandler|GetDebugMode) (Variables|CalendarPropHandler|GetDebugMode))
        (bind _returnvalue_2 (Audio|SpawnSoundAttached (Variables|Default|GetSound_InWorld) _output_get))
        (bind _output_get_2 (Variables|Default|SetInWorldSound _returnvalue_2))
        (Audio|Components|Audio|FadeIn _output_get_2 1.0 1.0 0.0 "Linear" :!class /Script/Engine.AudioComponent)
        (else
          (Utilities|FlowControl|Sequence
            (:then_0
              (Utilities|IsValid _shimmerfx
                (:Is%20Valid
                  (Components|DestroyComponent _shimmerfx)
                  (Development|PrintString "Shimmer Disabled" (Variables|CalendarPropHandler|GetDebugMode) (Variables|CalendarPropHandler|GetDebugMode)))))
            (:then_1
              (bind _inworldsound (Variables|Default|GetInWorldSound)
                (:then
                  (Audio|Components|Audio|FadeOut _inworldsound 0.25 0.0 "Linear" :!class /Script/Engine.AudioComponent))
                (:else)))))))))

(fn SpawnProp(Server)
  (for (_array_element _array_index) (Variables|Default|GetPropInfo)
    (bind _returnvalue (Game|SpawnActorfromClass (Utilities|Struct|BreakCalendarEventPropInfo _array_element) (.Transform_2_E937529F4F2C2D7D7A71F9B18165C447 (Utilities|Struct|BreakCalendarEventPropInfo _array_element)) "Undefined" "SelectDefaultAtRuntime" 0 0))
    (Utilities|Array|SetArrayElem (Variables|CalendarPropHandler|GetPropReferences) _array_index _returnvalue)
    (Development|PrintString "Spawned Prop!" (Variables|CalendarPropHandler|GetDebugMode) (Variables|CalendarPropHandler|GetDebugMode))
    (Variables|Default|SetPropHasBeenSpawned true)))

(fn ConstructionScript
  (Transformation|SetRelativeTransform (Variables|Default|GetDestructionBox) (Variables|CalendarPropHandler|GetDestructionBoxTransform) false false))
